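# Material function that outputs World Position Offset to billboard a mesh toward the camera and stretch it along world Z.
# UV direction from a center, scaled by each vertex's pivot distance, rebuilds the shape in camera space.
# Masked stretch, optionally scaled by object X scale, fades as the camera-to-object line nears vertical.
# AngleAlpha outputs camera-direction verticality; an alternative matrix-rotation branch is left unconnected.
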MaterialFunction /WildEstate/TechArt/MaterialFunctions/MF_SingleAxisWPOStretch
graph {nodes [
  n1 = Constant3Vector() {constant=<rgba 0 0 1 0>}
  n2 = Normalize(VectorInput=Subtract(A=CameraPositionWS(), B=ObjectPositionWS()))
  n3 = Abs(None=DotProduct(A=@n1, B=@n2))
  n4 = FunctionInput() {bUsePreviewValueAsDefault=true, inputName=Intensity, inputType=FunctionInput_Scalar, previewValue=<100 0 0 1>}
  n5 = FunctionInput(Preview=StaticBool() {value=true}) {
   bUsePreviewValueAsDefault=true
   desc="Scales intensity with object scale. (Referencing X Axis Scale).\r\n\r\nIntensity * Scale X = Final Intensity"
   inputName=RelativeToObjectScale
   inputType=FunctionInput_StaticBool
   sortPriority=2
   previewValue=<100 0 0 1>
  }
  n6 = StaticSwitch(
   True=Multiply(A=@n4, B=MaterialFunctionCall() {materialFunction=&ObjectScale}.Scale X)
   False=@n4
   Value=@n5
  )
  n7 = FunctionInput(Preview=VertexColor().R) {bUsePreviewValueAsDefault=true, desc=Mask where the stretching takes place, inputName=Mask, sortPriority=1, previewValue=<100 0 0 1>}
  n8 = Multiply(A=OneMinus(None=Saturate(None=@n3)), B=Multiply(A=Multiply(A=@n1, B=@n6), B=@n7))
  n9 = Normalize(VectorInput=Subtract(A=WorldPosition().XYZ, B=ObjectPositionWS()))
  n10 = CrossProduct(A=Constant3Vector() {constant=<rgba 0 0 1 0>}, B=@n2)
  n11 = Normalize(VectorInput=@n10)
  n12 = Normalize(VectorInput=CrossProduct(A=@n2, B=@n10))
  Add(
   A=@n8
   B=Subtract(
    A=MaterialFunctionCall(VectorToTransform=@n9, BasisX=@n2, BasisY=@n11, BasisZ=@n12) {materialFunction=&Transform3x3Matrix}.Transformed Vector
    B=@n9
   )
  ) {desc=WPO}
  CameraVectorWS()
  n13 = FunctionInput(Preview=TextureCoordinate()) {bUsePreviewValueAsDefault=true, inputName=UVs, inputType=FunctionInput_Vector2, sortPriority=10}
  n14 = ComponentMask(None=@n13) {g=true}
  n15 = Subtract(
   A=AppendVector(A=ComponentMask(None=@n13) {r=true}, B=OneMinus(None=@n14))
   B=FunctionInput(Preview=Constant2Vector() {g=0.5, r=0.5}) {bUsePreviewValueAsDefault=true, inputName=UV Center, inputType=FunctionInput_Vector2, sortPriority=11}
  ) {constB=0.5, desc=Make all verts equal distance to center}
  n16 = Max(A=SquareRoot(None=DotProduct(A=@n15, B=@n15))) {constB=0}
  n17 = WorldPosition()
  n18 = ObjectPositionWS()
  n19 = FunctionInput() {bUsePreviewValueAsDefault=true, inputName=Scale, inputType=FunctionInput_Scalar, sortPriority=12, previewValue=<1 1 1 1>}
  n20 = Multiply(A=Multiply(A=Divide(A=@n15, B=@n16), B=Distance(A=@n17.XYZ, B=@n18)), B=@n19)
  n21 = Constant()
  n22 = Add(A=Transform(None=AppendVector(A=@n20, B=@n21)) {transformSourceType=TRANSFORMSOURCE_Camera}, B=ObjectPositionWS())
  n23 = WorldPosition()
  FunctionOutput(None=Add(A=Subtract(A=@n22, B=@n23.XYZ), B=@n8)) {outputName=WPO}
  n24 = Constant3Vector() {constant=<rgba 0 0 1 0>}
  n25 = MaterialFunctionCall() {materialFunction=&CameraDirectionVector}
  FunctionOutput(None=Abs(None=DotProduct(A=@n24, B=@n25.Result))) {desc="Returns 1 when looking down the world Z axis, 0 when perpendicular.", outputName=AngleAlpha, sortPriority=1}
 ]}
