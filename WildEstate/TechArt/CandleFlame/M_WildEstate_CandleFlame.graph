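# Unlit translucent candle flame material that draws a procedural flame from exponential radial gradients, with no textures.
# An outer gradient minus an animated tip mask gives depth-faded opacity, while a core gradient blends two colors into emissive divided by eye adaptation.
# Vertices move through vertex-red-masked single-axis stretch, camera offset and a wind function.
# Position-offset layered sines produce flicker, and the stretch function exposes a view-angle alpha through reroutes.
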Material /WildEstate/TechArt/CandleFlame/M_WildEstate_CandleFlame
graph {
 properties {
  bAutomaticallySetUsageInEditor=false
  bUseMaterialAttributes=true
  blendMode=BLEND_Translucent
  shadingModel=MSM_Unlit
  usedShadingModels=Unlit
 }
 nodes [
  n1 = NamedRerouteUsage()
  n2 = MaterialFunctionCall(
   Offset Amount=ScalarParameter() {parameterName=CameraOffset}
   World Position=MaterialFunctionCall() {materialFunction=&ObjectPivotPoint}.Object Pivot Location
  ) {materialFunction=/Engine/Functions/Engine_MaterialFunctions02/WorldPositionOffset/CameraOffset}
  n3 = NamedRerouteUsage()
  n4 = NamedRerouteDeclaration(None=MaterialFunctionCall(UVs=TextureCoordinate(), "Rotation Angle (0-1)"=Constant() {r=0.375}) {materialFunction=&CustomRotator}.Rotated Values) {name=AdjustedUVs, nodeColor=<rgba 1 0.6328 0 1>}
  n5 = Constant2Vector() {g=0.85, r=2}
  n6 = Add(A=Multiply(A=@n4, B=@n5), B=Multiply(A=Divide(B=@n5), B=Constant2Vector() {r=-1}) {constB=-1})
  n7 = LinearInterpolate(A=@n6, B=@n4, Alpha=LinearInterpolate(Alpha=@n3) {constA=0.9})
  n8 = ScalarParameter() {defaultValue=0.5, parameterName=ShapeRadius}
  n9 = ScalarParameter() {defaultValue=2.33, parameterName=ShapeDensity}
  n10 = Time()
  n11 = Add(A=Sine(None=@n10) {period=0.93}, B=Sine(None=@n10) {period=2.1})
  n12 = Sine(None=Add(A=@n10) {constB=0.11}) {period=0.73}
  n13 = AppendVector(A=Multiply(A=Add(A=@n11, B=@n12), B=ScalarParameter() {defaultValue=0.02, parameterName=TipMotionIntensity}) {constB=0.01}, B=Constant())
  n14 = Frac(None=Add(A=@n4, B=Add(A=@n13) {constB=0.5}))
  n15 = ScalarParameter() {defaultValue=0.3, parameterName=SubRadius}
  n16 = ScalarParameter() {defaultValue=2.33, parameterName=SubDensity}
  n17 = OneMinus(None=MaterialFunctionCall(UV Channel=Add(A=@n4) {constB=0.5}) {materialFunction=&LinearGradient}.VGradient)
  n18 = Multiply(
   A=Multiply(
    A=MaterialFunctionCall(UVs=@n14, Radius=@n15, Density=@n16) {materialFunction=&RadialGradientExponential}.RadialGradientExponential
    B=Saturate(None=@n17)
   )
   B=ScalarParameter() {parameterName=SubtractiveMask}
  )
  n19 = NamedRerouteUsage()
  n20 = Saturate(None=Subtract(
    A=MaterialFunctionCall(UVs=@n7, Radius=@n8, Density=@n9) {materialFunction=&RadialGradientExponential}.RadialGradientExponential
    B=LinearInterpolate(A=@n18, Alpha=@n19) {constB=0}
   ))
  n21 = QualitySwitch(
   Default=DepthFade(Opacity=@n20, FadeDistance=ScalarParameter() {defaultValue=5, parameterName=DepthFade}) {fadeDistanceDefault=20}
   Low=@n20
   Medium=@n20
  )
  n22 = ScalarParameter() {defaultValue=1, parameterName=OpacityMaster}
  n23 = Add(
   A=LinearInterpolate(A=Constant2Vector() {g=-0.05}, Alpha=NamedRerouteUsage()) {constB=0}
   B=LinearInterpolate(A=@n6, B=@n4, Alpha=LinearInterpolate(Alpha=@n3) {constA=0.6})
  )
  n24 = ScalarParameter() {defaultValue=0.2, parameterName=CoreRadius}
  n25 = ScalarParameter() {defaultValue=2.33, parameterName=CoreDensity}
  n26 = Multiply(
   A=LinearInterpolate(
    A=VectorParameter() {parameterName=ColorOuter, defaultValue=<rgba 1 0.4004 0 1>}.RGB
    B=VectorParameter() {parameterName=ColorCore, defaultValue=<rgba 0.9672 1 0.0579 1>}.RGB
    Alpha=MaterialFunctionCall(UVs=@n23, Radius=@n24, Density=@n25) {desc=Core Mask, materialFunction=&RadialGradientExponential}.RadialGradientExponential
   )
   B=ScalarParameter() {defaultValue=1, parameterName=EmissiveMulti}
  )
  n27 = EyeAdaptation()
  MaterialFunctionCall!(In=MakeMaterialAttributes!(
    BaseColor
    Metallic
    Specular
    Roughness
    Anisotropy
    EmissiveColor=Divide(A=@n26, B=@n27.EyeAdaptation)
    Opacity=Multiply(A=@n21, B=@n22)
    OpacityMask
    Normal
    Tangent
    WorldPositionOffset=Add(A=@n1, B=@n2.Camera Offset)
    SubsurfaceColor
    ClearCoat
    ClearCoatRoughness
    AmbientOcclusion
    Refraction
    CustomizedUVs_0
    CustomizedUVs_1
    CustomizedUVs_2
    CustomizedUVs_3
    CustomizedUVs_4
    CustomizedUVs_5
    CustomizedUVs_6
    CustomizedUVs_7
    PixelDepthOffset
    ShadingModel
    Displacement
   )) {materialFunction=/WildEstateTech/Environment/BaseMaterials/MaterialFunctions/MF_WE_WPO_WindNew}
  n28 = ScalarParameter() {defaultValue=1.1, parameterName=IntensityFlickMulti}
  n29 = ScalarParameter() {defaultValue=100, parameterName=Intensity}
  n30 = NamedRerouteUsage()
  n31 = MaterialFunctionCall!(
   Intensity=LinearInterpolate(A=Multiply(A=@n28, B=@n29), B=@n29, Alpha=@n30)
   Mask=VertexColor().R
   RelativeToObjectScale
   UVs=NamedRerouteUsage()
   UV Center
   Scale=LinearInterpolate(Alpha=@n30) {constA=1, constB=1.25}
  ) {materialFunction=/WildEstate/TechArt/MaterialFunctions/MF_SingleAxisWPOStretch}
  NamedRerouteDeclaration(None=@n31.WPO) {name=WPO, nodeColor=<rgba 0.9453 0 1 1>}
  n32 = MaterialFunctionCall(Float3=ObjectPositionWS()) {materialFunction=&BreakOutFloat3Components}
  n33 = Add(A=Add(A=@n32.R, B=@n32.R), B=@n32.B)
  n34 = ScalarParameter() {defaultValue=0.25, group=WPO Sway, parameterName=TimeOffsetAmount}
  n35 = Multiply(A=Add(A=Divide(A=@n33, B=@n34), B=Time()), B=Multiply(A=ScalarParameter() {defaultValue=1, parameterName=Speed}) {constB=1.5}) {constB=2}
  n36 = Sine(None=Add(A=@n35) {constB=0.1}) {period=0.21}
  n37 = Sine(None=@n35) {period=0.29}
  n38 = Saturate(None=Multiply(A=ConstantBiasScale(None=@n36), B=ConstantBiasScale(None=@n37)))
  n39 = Sine(None=@n35) {period=2.5}
  n40 = Saturate(None=Multiply(A=ConstantBiasScale(None=@n39)) {constB=5})
  n41 = Sine(None=Add(A=@n35) {constB=0.33}) {period=1.71}
  n42 = Saturate(None=Multiply(A=ConstantBiasScale(None=@n41)) {constB=2})
  NamedRerouteDeclaration(None=LinearInterpolate(A=@n38, B=LinearInterpolate(Alpha=@n38) {constA=0.6}, Alpha=Multiply(A=@n40, B=@n42))) {name=Flicker, nodeColor=<rgba 1 0 0.6562 1>}
  NamedRerouteDeclaration(None=@n31.AngleAlpha) {name=IsParallel, nodeColor=<rgba 0.5938 0 1 1>}
 ]
}
